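# UI-domain material M_UI_DamageNumberFontOutline, named for the outline of damage number text.
# It holds a single vector parameter, Color, that instances can set to change the tint.
# The file shows no connections from that parameter to the material outputs and no other shading logic.
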Material /Sprites_ZeroPointSprite/UI/Materials/M_UI_DamageNumberFontOutline
graph {
 properties {materialDomain=MD_UI, pixelDepthOffsetMode=PDOM_Legacy, usedShadingModels=Default Lit}
 nodes [VectorParameter() {parameterName=Color}]
}
