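# Damage number widget for the Super Sprite variant, showing a yellow number with EMP, crit, damage-type and Super Sprite icons.
# Normal and crit animations pop, rise and fade, with crit adding a shear; the relevant finish event resets the widget.
# Shield breaks set IsBroken on the EMP icon material, cleared on reset.
# HUD scale changes recompute render scale from the damage-number setting, offset by HUD scale, and log it.
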WidgetBlueprint /Sprites_ZeroPointSprite/UI/WBP_DamageNumber_SuperSprite
animations [
 OnDamage {
  binding_count=1
  length_seconds=2.5
  object=&self:OnDamage
  sequence {bindings [SizeBox_0 {
     id=#1
     possessed_class=SizeBox
     tracks [
      {
       type=MovieSceneFloatTrack
       display_name=Render Opacity
       sections [{type=MovieSceneFloatSection, channels [None {type=MovieSceneScriptingFloatChannel, keys [[0,0.9],[3,1],[7,0.9],[10,0.8],[15,0.7],[30,0]]}]}]
      }
      {
       type=MovieScene2DTransformTrack
       sections [{
         type=MovieScene2DTransformSection
         end_frame=50
         start_frame=0
         channels [
          Translation.X {type=MovieSceneScriptingFloatChannel, const [0,1]}
          Translation.Y {type=MovieSceneScriptingFloatChannel, keys [[0,0],[15,0],[25,-10],[30,-40],[40,-90],[50,-120]]}
          Scale.X {type=MovieSceneScriptingFloatChannel, keys [[0,0.6],[3,0.7],[7,0.6]]}
          Scale.Y {type=MovieSceneScriptingFloatChannel, keys [[0,0.6],[3,0.7],[7,0.6]]}
         ]
        }]
      }
     ]
    }]}
 }
 OnDamage_Crit {
  binding_count=3
  length_seconds=2.5
  object=&self:OnDamage_Crit
  sequence {bindings [
    OverlaySlot_2 {id=#2, parent_id=#3, possessed_class=OverlaySlot, tracks []}
    SizeBox_0 {
     id=#4
     possessed_class=SizeBox
     tracks [
      {
       type=MovieSceneFloatTrack
       display_name=Render Opacity
       sections [{type=MovieSceneFloatSection, channels [None {type=MovieSceneScriptingFloatChannel, keys [[0,0.9],[3,1],[7,0.9],[10,0.8],[15,0.7],[30,0]]}]}]
      }
      {
       type=MovieScene2DTransformTrack
       sections [{
         type=MovieScene2DTransformSection
         end_frame=50
         start_frame=0
         channels [
          Translation.X {type=MovieSceneScriptingFloatChannel, const [0,1]}
          Translation.Y {type=MovieSceneScriptingFloatChannel, keys [[0,0],[15,0],[25,-10],[30,-40],[40,-90],[50,-120]]}
          Scale.X {type=MovieSceneScriptingFloatChannel, keys [[0,0.6],[2,0.78],[6,0.75],[10,0.6]]}
          Scale.Y {type=MovieSceneScriptingFloatChannel, keys [[0,0.6],[2,0.78],[6,0.75],[10,0.6]]}
         ]
        }]
      }
     ]
    }
    Text_Number {
     id=#3
     possessed_class=CommonTextBlock
     tracks [{
       type=MovieScene2DTransformTrack
       sections [{
         type=MovieScene2DTransformSection
         end_frame=30
         start_frame=0
         channels [
          Translation.X {type=MovieSceneScriptingFloatChannel, const [0,1]}
          Translation.Y {type=MovieSceneScriptingFloatChannel, const [0,1]}
          Angle {type=MovieSceneScriptingFloatChannel, const [0,1]}
          Scale.X {type=MovieSceneScriptingFloatChannel, const [1,1]}
          Scale.Y {type=MovieSceneScriptingFloatChannel, const [1,1]}
          Shear.X {type=MovieSceneScriptingFloatChannel, thinned_from=5, keys [[0,0],[2,-11],[25,-11],[30,0]]}
          Shear.Y {type=MovieSceneScriptingFloatChannel, const [0,1]}
         ]
        }]
      }]
    }
   ]}
 }
]
graphs [
 CritIconVisibilty {nodes [
   n1 = CritIconVisibilty!() -> (then, IsCrit:bool=~) {is=title K2Node_FunctionEntry}
   > SetVisibility(
    self=Get DamageTypeCrit() -> (DamageTypeCrit:Image Object Reference)
    InVisibility=Select(Option 0=Collapsed, Option 1=SelfHitTestInvisible, Index=@n1)
   )
  ]}
 EventGraph {nodes [
   n1 = Call On Finished Animating!(execute:Exec, self:WBP Damage Number Super Sprite Object Reference, DamageNumberWidget:WBP Damage Number Object Reference=Self-Reference()) -> (then:Exec) {is=title K2Node_CallDelegate}
   Reset(execute=[@n1.then, @n14.then], self:Common User Widget Damage Numbers Object Reference)
   > n2 = GetDynamicMaterial(self=Get DamageTypeIcon_EMP() -> (DamageTypeIcon_EMP:Image Object Reference))
   > Branch(Condition=IsValid(Object=@n2))
   > SetScalarParameterValue(self=@n2, ParameterName=IsBroken, Value=0)
   OnDamageAnimationFinished() {is=title K2Node_CustomEvent}
   > n3 = Branch(Condition=Get bIsPlayingCritAnimation() -> (bIsPlayingCritAnimation:bool))
   NotifyAnimationFinished(execute=@n3.else)
   n4 = OnHUDScaleChanged() {is=title K2Node_CustomEvent}
   n5 = GetHUDScale!(self:Fort Settings Context Object Reference=Get FortSettingsContext!() -> (ReturnValue:Fort Settings Context Object Reference) {is=title K2Node_GetBlueprintContextEx}) -> (ReturnValue:float)
   n6 = Branch(execute=[@n16.then, @n4.then], Condition="Equal (Float)"(A=@n5))
   > n7 = Set InverseHUDScaleVector(InverseHUDScaleVector:Vector 2D Structure=MakeVector2D(X=1, Y=1)) -> (Output_Get:Vector 2D Structure)
   n8 = GetHUDScale(self=Get FortSettingsContext())
   n9 = float + float(
    A=GetDamageNumbersScale!(self:Fort Settings Context Object Reference=Get FortSettingsContext()) -> (ReturnValue:float)
    B=float - float(A=1, B=@n8)
   )
   n10 = MakeVector2D(X=@n9, Y=@n9)
   Set DamageNumberScaleVector(execute=[@n7.then, @n13.then], DamageNumberScaleVector:Vector 2D Structure=@n10) -> (Output_Get:Vector 2D Structure)
   > SetRenderScale!(execute, self:Widget Object Reference, Scale:Vector 2D Structure=@n10) -> (then)
   n11 = GetDamageNumbersScale(self=Get FortSettingsContext())
   > FortBlueprintLog!(
    execute
    InString:String="To String (Text)"(InText=Format Text!(
      Format="LOCTEXT(\"WBP_DamageNumber: Setting render scale to {0}, HUD scale {1}, Damge Numbers scale {2}\")"
      0:double=@n9
      1:float=@n8
      2:float=@n11
     ) -> (Result))
   ) -> (then)
   n12 = float / float(A=1, B=@n5)
   n13 = Set InverseHUDScaleVector(execute=@n6.else, InverseHUDScaleVector=MakeVector2D(X=@n12, Y=@n12))
   OnDamageCritAnimationFinished() {is=title K2Node_CustomEvent}
   > Branch(Condition=Get bIsPlayingCritAnimation())
   > NotifyAnimationFinished()
   n14 = NotifyAnimationFinished() {is=title K2Node_CustomEvent}
   Event OnShieldBreak!() -> (OutputDelegate, then, bIsOvershield:bool) {is=title K2Node_Event}
   > n15 = GetDynamicMaterial(self=Get DamageTypeIcon_EMP())
   > Branch(Condition=IsValid(Object=@n15))
   > SetScalarParameterValue(self=@n15, ParameterName=IsBroken, Value=1)
   Event OnInitialized()
   > BindToAnimationFinished(Animation=Get OnDamage() -> (OnDamage:Widget Animation Object Reference), Delegate=Create Event())
   > BindToAnimationFinished(Animation=Get OnDamage_Crit() -> (OnDamage_Crit:Widget Animation Object Reference), Delegate=Create Event())
   > n16 = Bind Event to On HUDScale Changed(
    self:Fort HUDContext Object Reference=Get FortHUDContext!() -> (ReturnValue:Fort HUDContext Object Reference) {is=title K2Node_GetBlueprintContextEx}
    Delegate=@n4
   ) {is=title K2Node_AddDelegate}
  ]}
 GetDamageAmount {nodes [
   GetDamageAmount() {is=title K2Node_FunctionEntry}
   > Return Node!(execute, DamageAmount:double=Get Damage() -> (Damage:float))
  ]}
 GetDamagedActor {nodes [
   GetDamagedActor() {is=title K2Node_FunctionEntry}
   > Return Node!(execute, DamagedActor:Actor Object Reference=Get HitActor() -> (HitActor:Actor Object Reference))
  ]}
 GetSpawnTime {nodes [
   GetSpawnTime() {is=title K2Node_FunctionEntry}
   > Return Node!(execute, SpawnTime:double=Get SpawnTime() -> (SpawnTime:double))
  ]}
 OnFinishedAnimating {nodes [OnFinishedAnimating!() -> (then, DamageNumberWidget:WBP Damage Number Object Reference) {is=title K2Node_FunctionEntry}]}
 SetAnimatedVerticalPositionOffset {nodes [
   n1 = SetAnimatedVerticalPositionOffset!() -> (then, NewVerticalPosition:double=~) {is=title K2Node_FunctionEntry}
   > Set AnimatedVerticalPositionOffset(AnimatedVerticalPositionOffset:double=@n1) -> (Output_Get:double)
  ]}
 ShiftVertically {nodes [
   n1 = ShiftVertically!() -> (then, AdditionalVerticalOffset:double=~) {is=title K2Node_FunctionEntry}
   > Set AdditionalVerticalScreenOffset(AdditionalVerticalScreenOffset:float=float + float(A=@n1, B=Get AdditionalVerticalScreenOffset() -> (AdditionalVerticalScreenOffset:float))) -> (Output_Get:float)
  ]}
 UpdateWorldSpacePosition {nodes [
   n1 = UpdateWorldSpacePosition!() -> (then, WorldspacePos:Vector=~) {is=title K2Node_FunctionEntry}
   > Set WorldSpacePos(WorldSpacePos:Vector=@n1) -> (Output_Get:Vector)
  ]}
]
variables [
 AnimatedVerticalPositionOffset {type=number}
 HealthColorTemp {type=boolean}
 Lifetime {type=number, default=2}
 OnFinishedAnimating {}
]
widgets {
 info {
  inheritedWidgetCount=5
  parentClass=/Script/DamageNumbersUI.CommonUserWidget_DamageNumbers
  rootWidgetClass=&SizeBox
  widgetCount=10
 }
 widgets [
  SizeBox_0 {
   widgetClassPath=&SizeBox
   widget=SizeBox_0
   properties {
    bOverride_WidthOverride=true
    renderOpacity=3.9
    widthOverride=250
    renderTransform {angle=0, scale=<0.6 0.6>, shear=<0 0>, translation=<0 0>}
   }
  }
  HorizontalBox_0 {
   widgetClassPath=&HorizontalBox
   parent=SizeBox_0
   widget=HorizontalBox_0
   slot {class=SizeBoxSlot, properties {horizontalAlignment=HAlign_Left}}
  }
  DamageTypeIcon_EMP {
   widgetClassPath=&Image
   bInherited=true
   bIsVariable=true
   parent=HorizontalBox_0
   widget=DamageTypeIcon_EMP
   properties {
    visibility=HitTestInvisible
    brush {
     bIsDynamicallyLoaded=false
     drawAs=Image
     imageType=FullColor
     mirroring=NoMirror
     resourceName=None
     resourceObject=/Sprites_ZeroPointSprite/UI/Materials/M_UI_EmpGrenadeShieldIcon
     tiling=NoTile
     imageSize=<64 64>
     margin=<ltrb 0 0 0 0>
     outlineSettings {
      bUseBrushTransparency=false
      roundingType=HalfHeightRadius
      width=0
      color {colorUseRule=UseColor_Specified, specifiedColor=<rgba 0 0 0 0>}
      cornerRadii=<0 0 0 0>
     }
     tintColor {colorUseRule=UseColor_Specified, specifiedColor=<rgba 1 1 1 1>}
     uVRegion {bIsValid=false, max=<0 0>, min=<0 0>}
    }
    colorAndOpacity=<rgba 1 1 1 0.9>
    renderTransform {angle=0, scale=<1 1>, shear=<0 0>, translation=<-5 0>}
   }
   slot {class=HorizontalBoxSlot, properties {horizontalAlignment=HAlign_Left, verticalAlignment=VAlign_Top}}
  }
  Overlay_16 {
   widgetClassPath=&Overlay
   parent=HorizontalBox_0
   widget=Overlay_16
   properties {renderTransform {angle=0, scale=<1.2 1.2>, shear=<0 0>, translation=<0 0>}}
   slot {class=HorizontalBoxSlot, properties {horizontalAlignment=HAlign_Center}}
  }
  Text_Number_Stroke {
   widgetClassPath=&CommonTextBlock
   bInherited=true
   bIsVariable=true
   parent=Overlay_16
   widget=Text_Number_Stroke
   properties {
    justification=Right
    text=835
    visibility=Collapsed
    font {
     bForceMonospaced=false
     bMaterialIsStencil=false
     fontMaterial=null
     fontObject=&Burbank
     letterSpacing=100
     monospacedWidth=1
     size=20
     skewAmount=0
     typefaceFontName=Burbank-Black
     outlineSettings {
      bApplyOutlineToDropShadows=false
      bMiteredCorners=false
      bSeparateFillAlpha=false
      outlineMaterial=/Sprites_ZeroPointSprite/UI/Materials/MI_UI_DamageNumberFontOutline
      outlineSize=1
      outlineColor=<rgba 1 1 1 1>
     }
    }
    shadowOffset=<0 0>
   }
   slot {class=OverlaySlot, properties {horizontalAlignment=HAlign_Center, verticalAlignment=VAlign_Center}}
  }
  Text_Number {
   widgetClassPath=&CommonTextBlock
   bInherited=true
   bIsVariable=true
   parent=Overlay_16
   widget=Text_Number
   properties {
    justification=Right
    text=835
    colorAndOpacity {colorUseRule=UseColor_Specified, specifiedColor=<rgba 0.8333 0.8238 0.0046 1>}
    font {
     bForceMonospaced=false
     bMaterialIsStencil=false
     fontMaterial=null
     fontObject=&Burbank
     letterSpacing=120
     monospacedWidth=1
     size=36
     skewAmount=0
     typefaceFontName=Burbank-Black
     outlineSettings {
      bApplyOutlineToDropShadows=false
      bMiteredCorners=false
      bSeparateFillAlpha=false
      outlineMaterial=/Sprites_ZeroPointSprite/UI/Materials/MI_UI_DamageNumbersInnerStroke
      outlineSize=2
      outlineColor=<rgba 1 1 1 1>
     }
    }
    shadowColorAndOpacity=<rgba 0 0.7448 0.7448 1>
    shadowOffset=<0 0>
   }
   slot {class=OverlaySlot, properties {horizontalAlignment=HAlign_Right, verticalAlignment=VAlign_Fill}}
  }
  HorizontalBox_2 {
   widgetClassPath=&HorizontalBox
   parent=HorizontalBox_0
   widget=HorizontalBox_2
   slot {class=HorizontalBoxSlot, properties {horizontalAlignment=HAlign_Left, padding=<ltrb 0 3 0 0>}}
  }
  DamageTypeCrit {
   widgetClassPath=&Image
   bInherited=true
   bIsVariable=true
   parent=HorizontalBox_2
   widget=DamageTypeCrit
   properties {
    visibility=SelfHitTestInvisible
    brush {
     bIsDynamicallyLoaded=false
     drawAs=Image
     imageType=FullColor
     mirroring=NoMirror
     resourceName=None
     resourceObject=/Sprites_ZeroPointSprite/UI/Materials/MI_UI_DamageTypeIcon_Crit
     tiling=NoTile
     imageSize=<40 40>
     margin=<ltrb 0 0 0 0>
     outlineSettings {
      bUseBrushTransparency=false
      roundingType=HalfHeightRadius
      width=0
      color {colorUseRule=UseColor_Specified, specifiedColor=<rgba 0 0 0 0>}
      cornerRadii=<0 0 0 0>
     }
     tintColor {colorUseRule=UseColor_Specified, specifiedColor=<rgba 1 1 1 0.9>}
     uVRegion {bIsValid=false, max=<0 0>, min=<0 0>}
    }
    colorAndOpacity=<rgba 1 1 1 0.9>
   }
   slot {class=HorizontalBoxSlot, properties {horizontalAlignment=HAlign_Left, verticalAlignment=VAlign_Top, padding=<ltrb 2 0 -18 0>}}
  }
  DamageTypeIcon {
   widgetClassPath=&Image
   bInherited=true
   bIsVariable=true
   parent=HorizontalBox_2
   widget=DamageTypeIcon
   properties {
    visibility=HitTestInvisible
    brush {
     bIsDynamicallyLoaded=false
     drawAs=Image
     imageType=FullColor
     mirroring=NoMirror
     resourceName=None
     resourceObject=/Sprites_ZeroPointSprite/UI/Materials/MI_UI_DamageTypeIcon
     tiling=NoTile
     imageSize=<40 40>
     margin=<ltrb 0 0 0 0>
     outlineSettings {
      bUseBrushTransparency=false
      roundingType=HalfHeightRadius
      width=0
      color {colorUseRule=UseColor_Specified, specifiedColor=<rgba 0 0 0 0>}
      cornerRadii=<0 0 0 0>
     }
     tintColor {colorUseRule=UseColor_Specified, specifiedColor=<rgba 1 1 1 1>}
     uVRegion {bIsValid=false, max=<0 0>, min=<0 0>}
    }
    colorAndOpacity=<rgba 1 1 1 0.9>
   }
   slot {class=HorizontalBoxSlot, properties {horizontalAlignment=HAlign_Left, verticalAlignment=VAlign_Top, padding=<ltrb 6 0 0 0>}}
  }
  DamageSuperSpriteIcon {
   widgetClassPath=&Image
   bIsVariable=true
   parent=HorizontalBox_2
   widget=DamageSuperSpriteIcon
   properties {
    visibility=HitTestInvisible
    brush {
     bIsDynamicallyLoaded=false
     drawAs=Image
     imageType=FullColor
     mirroring=NoMirror
     resourceName=None
     resourceObject=&MI_UI_DamageTypeIcon
     tiling=NoTile
     imageSize=<40 40>
     margin=<ltrb 0 0 0 0>
     outlineSettings {
      bUseBrushTransparency=false
      roundingType=HalfHeightRadius
      width=0
      color {colorUseRule=UseColor_Specified, specifiedColor=<rgba 0 0 0 0>}
      cornerRadii=<0 0 0 0>
     }
     tintColor {colorUseRule=UseColor_Specified, specifiedColor=<rgba 1 0.9021 0 1>}
     uVRegion {bIsValid=false, max=<0 0>, min=<0 0>}
    }
    colorAndOpacity=<rgba 1 1 1 0.9>
   }
   slot {class=HorizontalBoxSlot, properties {horizontalAlignment=HAlign_Left, verticalAlignment=VAlign_Top, padding=<ltrb 6 0 0 0>}}
  }
 ]
}
